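# Animation Blueprint for the F_MED_CoreStreet female player body cosmetic, adding secondary physics motion to necklace and jacket bones.
# Two long Anim Dynamics chains start from cached Base_pose, one directly and the other through the F_MED_CoreStreet Control Rig.
# Each node uses angular springs and gravity, with planar or spherical limits defined around spine, chest or pelvis bones.
# The chains differ in tuning, and the direct one simulates jacket bones in world space.
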AnimBlueprint /BRCosmetics/Characters/Player/Female/Medium/Bodies/F_MED_CoreStreet/Meshes/F_MED_CoreStreet_AnimBP
target_skeleton=&/F_MED_CoreStreet_Skeleton
graphs [
 AnimGraph {nodes [
   n1 = Anim Dynamics - Bone: dyn_neckalce_01(ComponentPose=Local To Component(LocalPose=Use cached pose 'Base_pose'()), Alpha=1, ExternalForce="0,0,0") {
    angularBiasOverride=0.999
    angularDampingOverride=0.999
    angularSpringConstant=1000
    bAngularSpring=true
    bOverrideAngularBias=true
    bOverrideAngularDamping=true
    bUseGravityOverride=true
    alphaBoolBlend {blendOption=Linear}
    boundBone {boneName=dyn_neckalce_01}
    componentAppliedLinearAccClamp=<400 400 400>
    componentLinearAccScale=<0.1 0.1 0.1>
    componentLinearVelScale=<0.2 0.2 0.2>
    gravityOverride {z=-980}
    physicsBodyDefinitions [{
      boundBone {boneName=dyn_neckalce_01}
      boxExtents=<15 15 15>
      constraintSetup {angularLimitsMax=<0 0 2>, angularLimitsMin=<0 0 2>, angularTarget=<0 0 0>}
      localJointOffset=<0.4 0 0>
     }]
    planarLimits [{drivingBone {boneName=spine_05}, planeTransform {location=<13 0 0>, rotation=<rot 0 -30.0001 90.0001>}}]
   }
   n2 = Anim Dynamics - Bone: dyn_neckalce_02(ComponentPose=@n1, Alpha=1, ExternalForce="0,0,0") {
    angularBiasOverride=0.999
    angularDampingOverride=0.999
    angularSpringConstant=1000
    bAngularSpring=true
    bOverrideAngularBias=true
    bOverrideAngularDamping=true
    bUseGravityOverride=true
    alphaBoolBlend {blendOption=Linear}
    boundBone {boneName=dyn_neckalce_02}
    componentAppliedLinearAccClamp=<400 400 400>
    componentLinearAccScale=<0.1 0.1 0.1>
    componentLinearVelScale=<0.2 0.2 0.2>
    gravityOverride {z=-980}
    physicsBodyDefinitions [{
      boundBone {boneName=dyn_neckalce_02}
      boxExtents=<15 15 15>
      constraintSetup {angularLimitsMax=<0 0 0>, angularLimitsMin=<0 0 0>, angularTarget=<0 0 0>}
      localJointOffset=<0.4 0 0>
     }]
    planarLimits [{drivingBone {boneName=spine_05}, planeTransform {location=<13 0 0>, rotation=<rot 0 -30.0001 90.0001>}}]
   }
   n3 = Anim Dynamics - Bone: dyn_neckalce_03(ComponentPose=@n2, Alpha=1, ExternalForce="0,0,0") {
    angularBiasOverride=0.999
    angularDampingOverride=0.999
    angularSpringConstant=1000
    bAngularSpring=true
    bOverrideAngularBias=true
    bOverrideAngularDamping=true
    bUseGravityOverride=true
    alphaBoolBlend {blendOption=Linear}
    boundBone {boneName=dyn_neckalce_03}
    componentAppliedLinearAccClamp=<400 400 400>
    componentLinearAccScale=<0.1 0.1 0.1>
    componentLinearVelScale=<0.2 0.2 0.2>
    gravityOverride {z=-980}
    physicsBodyDefinitions [{
      boundBone {boneName=dyn_neckalce_03}
      boxExtents=<15 15 15>
      constraintSetup {angularLimitsMax=<0 15 2>, angularLimitsMin=<0 -15 2>, angularTarget=<0 0 0>}
      localJointOffset=<0.4 0 0>
     }]
    planarLimits [{drivingBone {boneName=spine_05}, planeTransform {location=<13 0 0>, rotation=<rot 0 -30.0001 90.0001>}}]
   }
   n4 = Anim Dynamics - Bone: dyn_neckalce_04(ComponentPose=@n3, Alpha=1, ExternalForce="0,0,0") {
    angularBiasOverride=0.999
    angularDampingOverride=0.999
    angularSpringConstant=1000
    bAngularSpring=true
    bOverrideAngularBias=true
    bOverrideAngularDamping=true
    bUseGravityOverride=true
    alphaBoolBlend {blendOption=Linear}
    boundBone {boneName=dyn_neckalce_04}
    componentAppliedLinearAccClamp=<400 400 400>
    componentLinearAccScale=<0.1 0.1 0.1>
    componentLinearVelScale=<0.2 0.2 0.2>
    gravityOverride {z=-980}
    physicsBodyDefinitions [{
      boundBone {boneName=dyn_neckalce_04}
      boxExtents=<15 15 15>
      constraintSetup {angularLimitsMax=<0 15 5>, angularLimitsMin=<0 -15 3>, angularTarget=<0 0 0>}
      localJointOffset=<0.4 0 0>
     }]
    planarLimits [{drivingBone {boneName=spine_05}, planeTransform {location=<13 0 0>, rotation=<rot 0 -25.4743 90>}}]
    sphericalLimits [
     {limitRadius=6.7416, drivingBone {boneName=deform_pec_l}, sphereLocalOffset=<-0.3528 -0.1821 0.7624>}
     {limitRadius=6.7416, drivingBone {boneName=deform_pec_r}, sphereLocalOffset=<0.4557 -0.1503 -0.6637>}
    ]
   }
   n5 = Anim Dynamics - Bone: dyn_jacket_fr_l_01(ComponentPose=@n4, Alpha=1, ExternalForce="0,0,0") {
    angularBiasOverride=0.999
    angularDampingOverride=0.999
    angularSpringConstant=300
    bAngularSpring=true
    bOverrideAngularBias=true
    bOverrideAngularDamping=true
    bUseGravityOverride=true
    bUsePlanarLimit=false
    simulationSpace=World
    alphaBoolBlend {blendOption=Linear}
    boundBone {boneName=dyn_jacket_fr_l_01}
    componentAppliedLinearAccClamp=<600 600 600>
    componentLinearAccScale=<0.1 0.1 0.1>
    componentLinearVelScale=<0.1 0.1 0.1>
    gravityOverride {z=-980}
    physicsBodyDefinitions [{
      boundBone {boneName=dyn_jacket_fr_l_01}
      boxExtents=<10 10 10>
      constraintSetup {angularLimitsMax=<0 5 10>, angularLimitsMin=<0 -5 0>, angularTarget=<0 0 0>}
      localJointOffset=<1 0 0>
     }]
    sphericalLimits [
     {limitRadius=10, drivingBone {boneName=spine_04}, sphereLocalOffset=<0 0 -1>}
     {limitRadius=10, drivingBone {boneName=spine_03}, sphereLocalOffset=<0 0 -1>}
     {limitRadius=10, drivingBone {boneName=spine_02}, sphereLocalOffset=<0 0 -1>}
    ]
   }
   n6 = Anim Dynamics - Bone: dyn_jacket_fr_l_02(ComponentPose=@n5, Alpha=1, ExternalForce="0,0,0") {
    angularBiasOverride=0.999
    angularDampingOverride=0.999
    angularSpringConstant=300
    bAngularSpring=true
    bOverrideAngularBias=true
    bOverrideAngularDamping=true
    bUseGravityOverride=true
    bUsePlanarLimit=false
    simulationSpace=World
    alphaBoolBlend {blendOption=Linear}
    boundBone {boneName=dyn_jacket_fr_l_02}
    componentAppliedLinearAccClamp=<600 600 600>
    componentLinearAccScale=<0.1 0.1 0.1>
    componentLinearVelScale=<0.1 0.1 0.1>
    gravityOverride {z=-980}
    physicsBodyDefinitions [{
      boundBone {boneName=dyn_jacket_fr_l_02}
      boxExtents=<10 10 10>
      constraintSetup {angularLimitsMax=<0 5 30>, angularLimitsMin=<0 -5 0>, angularTarget=<0 0 0>}
      localJointOffset=<1 0 0>
     }]
    sphericalLimits [
     {limitRadius=10, drivingBone {boneName=spine_04}, sphereLocalOffset=<0 0 -1>}
     {limitRadius=10, drivingBone {boneName=spine_03}, sphereLocalOffset=<0 0 -1>}
     {limitRadius=10, drivingBone {boneName=spine_02}, sphereLocalOffset=<0 0 -1>}
    ]
   }
   n7 = Anim Dynamics - Bone: dyn_jacket_fr_l_03(ComponentPose=@n6, Alpha=1, ExternalForce="0,0,0") {
    angularBiasOverride=0.999
    angularDampingOverride=0.999
    angularSpringConstant=300
    bAngularSpring=true
    bOverrideAngularBias=true
    bOverrideAngularDamping=true
    bUseGravityOverride=true
    bUsePlanarLimit=false
    simulationSpace=World
    alphaBoolBlend {blendOption=Linear}
    boundBone {boneName=dyn_jacket_fr_l_03}
    componentAppliedLinearAccClamp=<600 600 600>
    componentLinearAccScale=<0.1 0.1 0.1>
    componentLinearVelScale=<0.1 0.1 0.1>
    gravityOverride {z=-980}
    physicsBodyDefinitions [{
      boundBone {boneName=dyn_jacket_fr_l_03}
      boxExtents=<10 10 10>
      constraintSetup {angularLimitsMax=<0 5 50>, angularLimitsMin=<0 -5 -10>, angularTarget=<0 0 0>}
      localJointOffset=<1 0 0>
     }]
    sphericalLimits [
     {limitRadius=10, drivingBone {boneName=spine_04}, sphereLocalOffset=<0 0 -1>}
     {limitRadius=10, drivingBone {boneName=spine_03}, sphereLocalOffset=<0 0 -1>}
     {limitRadius=10, drivingBone {boneName=spine_02}, sphereLocalOffset=<0 0 -1>}
    ]
   }
   n8 = Anim Dynamics - Bone: dyn_jacket_fr_side_l_01(ComponentPose=@n7, Alpha=1, ExternalForce="0,0,0") {
    angularBiasOverride=0.999
    angularDampingOverride=0.999
    angularSpringConstant=300
    bAngularSpring=true
    bOverrideAngularBias=true
    bOverrideAngularDamping=true
    bUseGravityOverride=true
    bUsePlanarLimit=false
    bUseSphericalLimits=true
    simulationSpace=World
    alphaBoolBlend {blendOption=Linear}
    boundBone {boneName=dyn_jacket_fr_side_l_01}
    componentAppliedLinearAccClamp=<600 600 600>
    componentLinearAccScale=<0.1 0.1 0.1>
    componentLinearVelScale=<0.1 0.1 0.1>
    gravityOverride {z=-980}
    physicsBodyDefinitions [{
      boundBone {boneName=dyn_jacket_fr_side_l_01}
      boxExtents=<10 10 10>
      constraintSetup {angularLimitsMax=<0 5 10>, angularLimitsMin=<0 -5 0>, angularTarget=<0 0 0>}
      localJointOffset=<1 0 0>
     }]
    sphericalLimits [
     {limitRadius=10, drivingBone {boneName=spine_04}, sphereLocalOffset=<0 0 -1>}
     {limitRadius=10, drivingBone {boneName=spine_03}, sphereLocalOffset=<0 0 -1>}
     {limitRadius=10, drivingBone {boneName=spine_02}, sphereLocalOffset=<0 0 -1>}
     {limitRadius=13, drivingBone {boneName=pelvis}, sphereLocalOffset=<0 0 0>}
    ]
   }
   n9 = Anim Dynamics - Bone: dyn_jacket_fr_side_l_02(ComponentPose=@n8, Alpha=1, ExternalForce="0,0,0") {
    angularBiasOverride=0.999
    angularDampingOverride=0.999
    angularSpringConstant=300
    bAngularSpring=true
    bOverrideAngularBias=true
    bOverrideAngularDamping=true
    bUseGravityOverride=true
    bUsePlanarLimit=false
    bUseSphericalLimits=true
    simulationSpace=World
    alphaBoolBlend {blendOption=Linear}
    boundBone {boneName=dyn_jacket_fr_side_l_02}
    componentAppliedLinearAccClamp=<600 600 600>
    componentLinearAccScale=<0.1 0.1 0.1>
    componentLinearVelScale=<0.1 0.1 0.1>
    gravityOverride {z=-980}
    physicsBodyDefinitions [{
      boundBone {boneName=dyn_jacket_fr_side_l_02}
      boxExtents=<10 10 10>
      constraintSetup {angularLimitsMax=<0 5 25>, angularLimitsMin=<0 -5 -5>, angularTarget=<0 0 0>}
      localJointOffset=<1 0 0>
     }]
    sphericalLimits [
     {limitRadius=10, drivingBone {boneName=spine_04}, sphereLocalOffset=<0 0 -1>}
     {limitRadius=10, drivingBone {boneName=spine_03}, sphereLocalOffset=<0 0 -1>}
     {limitRadius=10, drivingBone {boneName=spine_02}, sphereLocalOffset=<0 0 -1>}
     {limitRadius=13, drivingBone {boneName=pelvis}, sphereLocalOffset=<0 0 0>}
    ]
   }
   n10 = Anim Dynamics - Bone: dyn_jacket_side_l_01(ComponentPose=@n9, Alpha=1, ExternalForce="0,0,0") {
    angularBiasOverride=0.999
    angularDampingOverride=0.999
    angularSpringConstant=300
    bAngularSpring=true
    bOverrideAngularBias=true
    bOverrideAngularDamping=true
    bUseGravityOverride=true
    bUsePlanarLimit=false
    bUseSphericalLimits=true
    simulationSpace=World
    alphaBoolBlend {blendOption=Linear}
    boundBone {boneName=dyn_jacket_side_l_01}
    componentAppliedLinearAccClamp=<600 600 600>
    componentLinearAccScale=<0.1 0.1 0.1>
    componentLinearVelScale=<0.1 0.1 0.1>
    gravityOverride {z=-980}
    physicsBodyDefinitions [{
      boundBone {boneName=dyn_jacket_side_l_01}
      boxExtents=<10 10 10>
      constraintSetup {angularLimitsMax=<0 5 10>, angularLimitsMin=<0 -5 0>, angularTarget=<0 0 0>}
      localJointOffset=<1 0 0>
     }]
    sphericalLimits [
     {limitRadius=10, drivingBone {boneName=spine_04}, sphereLocalOffset=<0 0 -1>}
     {limitRadius=10, drivingBone {boneName=spine_03}, sphereLocalOffset=<0 0 -1>}
     {limitRadius=10, drivingBone {boneName=spine_02}, sphereLocalOffset=<0 0 -1>}
     {limitRadius=13, drivingBone {boneName=pelvis}, sphereLocalOffset=<0 0 0>}
    ]
   }
   n11 = Anim Dynamics - Bone: dyn_jacket_fr_side_l_02(ComponentPose=@n10, Alpha=1, ExternalForce="0,0,0") {
    angularBiasOverride=0.999
    angularDampingOverride=0.999
    angularSpringConstant=300
    bAngularSpring=true
    bOverrideAngularBias=true
    bOverrideAngularDamping=true
    bUseGravityOverride=true
    bUsePlanarLimit=false
    bUseSphericalLimits=true
    simulationSpace=World
    alphaBoolBlend {blendOption=Linear}
    boundBone {boneName=dyn_jacket_fr_side_l_02}
    componentAppliedLinearAccClamp=<600 600 600>
    componentLinearAccScale=<0.1 0.1 0.1>
    componentLinearVelScale=<0.1 0.1 0.1>
    gravityOverride {z=-980}
    physicsBodyDefinitions [{
      boundBone {boneName=dyn_jacket_fr_side_l_02}
      boxExtents=<10 10 10>
      constraintSetup {angularLimitsMax=<0 5 20>, angularLimitsMin=<0 -5 0>, angularTarget=<0 0 0>}
      localJointOffset=<1 0 0>
     }]
    sphericalLimits [
     {limitRadius=10, drivingBone {boneName=spine_04}, sphereLocalOffset=<0 0 -1>}
     {limitRadius=10, drivingBone {boneName=spine_03}, sphereLocalOffset=<0 0 -1>}
     {limitRadius=10, drivingBone {boneName=spine_02}, sphereLocalOffset=<0 0 -1>}
     {limitRadius=13, drivingBone {boneName=pelvis}, sphereLocalOffset=<0 0 0>}
    ]
   }
   n12 = Anim Dynamics - Bone: dyn_jacket_fr_r_01(ComponentPose=@n11, Alpha=1, ExternalForce="0,0,0") {
    angularBiasOverride=0.999
    angularDampingOverride=0.999
    angularSpringConstant=300
    bAngularSpring=true
    bOverrideAngularBias=true
    bOverrideAngularDamping=true
    bUseGravityOverride=true
    bUsePlanarLimit=false
    simulationSpace=World
    alphaBoolBlend {blendOption=Linear}
    boundBone {boneName=dyn_jacket_fr_r_01}
    componentAppliedLinearAccClamp=<600 600 600>
    componentLinearAccScale=<0.1 0.1 0.1>
    componentLinearVelScale=<0.1 0.1 0.1>
    gravityOverride {z=-980}
    physicsBodyDefinitions [{
      boundBone {boneName=dyn_jacket_fr_r_01}
      boxExtents=<10 10 10>
      constraintSetup {angularLimitsMax=<0 5 10>, angularLimitsMin=<0 -5 0>, angularTarget=<0 0 0>}
      localJointOffset=<-1 0 0>
     }]
    sphericalLimits [
     {limitRadius=10, drivingBone {boneName=spine_04}, sphereLocalOffset=<0 0 1>}
     {limitRadius=10, drivingBone {boneName=spine_03}, sphereLocalOffset=<0 0 1>}
     {limitRadius=10, drivingBone {boneName=spine_02}, sphereLocalOffset=<0 0 1>}
    ]
   }
   n13 = Anim Dynamics - Bone: dyn_jacket_fr_r_02(ComponentPose=@n12, Alpha=1, ExternalForce="0,0,0") {
    angularBiasOverride=0.999
    angularDampingOverride=0.999
    angularSpringConstant=300
    bAngularSpring=true
    bOverrideAngularBias=true
    bOverrideAngularDamping=true
    bUseGravityOverride=true
    bUsePlanarLimit=false
    simulationSpace=World
    alphaBoolBlend {blendOption=Linear}
    boundBone {boneName=dyn_jacket_fr_r_02}
    componentAppliedLinearAccClamp=<600 600 600>
    componentLinearAccScale=<0.1 0.1 0.1>
    componentLinearVelScale=<0.1 0.1 0.1>
    gravityOverride {z=-980}
    physicsBodyDefinitions [{
      boundBone {boneName=dyn_jacket_fr_r_02}
      boxExtents=<10 10 10>
      constraintSetup {angularLimitsMax=<0 5 30>, angularLimitsMin=<0 -5 0>, angularTarget=<0 0 0>}
      localJointOffset=<-1 0 0>
     }]
    sphericalLimits [
     {limitRadius=10, drivingBone {boneName=spine_04}, sphereLocalOffset=<0 0 1>}
     {limitRadius=10, drivingBone {boneName=spine_03}, sphereLocalOffset=<0 0 1>}
     {limitRadius=10, drivingBone {boneName=spine_02}, sphereLocalOffset=<0 0 1>}
    ]
   }
   n14 = Anim Dynamics - Bone: dyn_jacket_fr_r_03(ComponentPose=@n13, Alpha=1, ExternalForce="0,0,0") {
    angularBiasOverride=0.999
    angularDampingOverride=0.999
    angularSpringConstant=300
    bAngularSpring=true
    bOverrideAngularBias=true
    bOverrideAngularDamping=true
    bUseGravityOverride=true
    bUsePlanarLimit=false
    simulationSpace=World
    alphaBoolBlend {blendOption=Linear}
    boundBone {boneName=dyn_jacket_fr_r_03}
    componentAppliedLinearAccClamp=<600 600 600>
    componentLinearAccScale=<0.1 0.1 0.1>
    componentLinearVelScale=<0.1 0.1 0.1>
    gravityOverride {z=-980}
    physicsBodyDefinitions [{
      boundBone {boneName=dyn_jacket_fr_r_03}
      boxExtents=<10 10 10>
      constraintSetup {angularLimitsMax=<0 5 50>, angularLimitsMin=<0 -5 0>, angularTarget=<0 0 0>}
      localJointOffset=<-1 0 0>
     }]
    sphericalLimits [
     {limitRadius=10, drivingBone {boneName=spine_04}, sphereLocalOffset=<0 0 1>}
     {limitRadius=10, drivingBone {boneName=spine_03}, sphereLocalOffset=<0 0 1>}
     {limitRadius=10, drivingBone {boneName=spine_02}, sphereLocalOffset=<0 0 1>}
    ]
   }
   n15 = Anim Dynamics - Bone: dyn_jacket_fr_side_r_01(ComponentPose=@n14, Alpha=1, ExternalForce="0,0,0") {
    angularBiasOverride=0.999
    angularDampingOverride=0.999
    angularSpringConstant=300
    bAngularSpring=true
    bOverrideAngularBias=true
    bOverrideAngularDamping=true
    bUseGravityOverride=true
    bUsePlanarLimit=false
    simulationSpace=World
    alphaBoolBlend {blendOption=Linear}
    boundBone {boneName=dyn_jacket_fr_side_r_01}
    componentAppliedLinearAccClamp=<600 600 600>
    componentLinearAccScale=<0.1 0.1 0.1>
    componentLinearVelScale=<0.1 0.1 0.1>
    gravityOverride {z=-980}
    physicsBodyDefinitions [{
      boundBone {boneName=dyn_jacket_fr_side_r_01}
      boxExtents=<10 10 10>
      constraintSetup {angularLimitsMax=<0 5 10>, angularLimitsMin=<0 -5 0>, angularTarget=<0 0 0>}
      localJointOffset=<-1 0 0>
     }]
    sphericalLimits [
     {limitRadius=10, drivingBone {boneName=spine_04}, sphereLocalOffset=<0 0 1>}
     {limitRadius=10, drivingBone {boneName=spine_03}, sphereLocalOffset=<0 0 1>}
     {limitRadius=10, drivingBone {boneName=spine_02}, sphereLocalOffset=<0 0 1>}
    ]
   }
   n16 = Anim Dynamics - Bone: dyn_jacket_fr_side_r_02(ComponentPose=@n15, Alpha=1, ExternalForce="0,0,0") {
    angularBiasOverride=0.999
    angularDampingOverride=0.999
    angularSpringConstant=300
    bAngularSpring=true
    bOverrideAngularBias=true
    bOverrideAngularDamping=true
    bUseGravityOverride=true
    bUsePlanarLimit=false
    simulationSpace=World
    alphaBoolBlend {blendOption=Linear}
    boundBone {boneName=dyn_jacket_fr_side_r_02}
    componentAppliedLinearAccClamp=<600 600 600>
    componentLinearAccScale=<0.1 0.1 0.1>
    componentLinearVelScale=<0.1 0.1 0.1>
    gravityOverride {z=-980}
    physicsBodyDefinitions [{
      boundBone {boneName=dyn_jacket_fr_side_r_02}
      boxExtents=<10 10 10>
      constraintSetup {angularLimitsMax=<0 5 25>, angularLimitsMin=<0 -5 0>, angularTarget=<0 0 0>}
      localJointOffset=<-1 0 0>
     }]
    sphericalLimits [
     {limitRadius=10, drivingBone {boneName=spine_04}, sphereLocalOffset=<0 0 1>}
     {limitRadius=10, drivingBone {boneName=spine_03}, sphereLocalOffset=<0 0 1>}
     {limitRadius=10, drivingBone {boneName=spine_02}, sphereLocalOffset=<0 0 1>}
    ]
   }
   n17 = Anim Dynamics - Bone: dyn_jacket_side_r_01(ComponentPose=@n16, Alpha=1, ExternalForce="0,0,0") {
    angularBiasOverride=0.999
    angularDampingOverride=0.999
    angularSpringConstant=300
    bAngularSpring=true
    bOverrideAngularBias=true
    bOverrideAngularDamping=true
    bUseGravityOverride=true
    bUsePlanarLimit=false
    simulationSpace=World
    alphaBoolBlend {blendOption=Linear}
    boundBone {boneName=dyn_jacket_side_r_01}
    componentAppliedLinearAccClamp=<600 600 600>
    componentLinearAccScale=<0.1 0.1 0.1>
    componentLinearVelScale=<0.1 0.1 0.1>
    gravityOverride {z=-980}
    physicsBodyDefinitions [{
      boundBone {boneName=dyn_jacket_side_r_01}
      boxExtents=<10 10 10>
      constraintSetup {angularLimitsMax=<0 5 30>, angularLimitsMin=<0 -5 0>, angularTarget=<0 0 0>}
      localJointOffset=<-1 0 0>
     }]
    sphericalLimits [
     {limitRadius=10, drivingBone {boneName=spine_04}, sphereLocalOffset=<0 0 1>}
     {limitRadius=10, drivingBone {boneName=spine_03}, sphereLocalOffset=<0 0 1>}
     {limitRadius=10, drivingBone {boneName=spine_02}, sphereLocalOffset=<0 0 1>}
    ]
   }
   n18 = Anim Dynamics - Bone: dyn_jacket_side_r_02(ComponentPose=@n17, Alpha=1, ExternalForce="0,0,0") {
    angularBiasOverride=0.999
    angularDampingOverride=0.999
    angularSpringConstant=300
    bAngularSpring=true
    bOverrideAngularBias=true
    bOverrideAngularDamping=true
    bUseGravityOverride=true
    bUsePlanarLimit=false
    simulationSpace=World
    alphaBoolBlend {blendOption=Linear}
    boundBone {boneName=dyn_jacket_side_r_02}
    componentAppliedLinearAccClamp=<600 600 600>
    componentLinearAccScale=<0.1 0.1 0.1>
    componentLinearVelScale=<0.1 0.1 0.1>
    gravityOverride {z=-980}
    physicsBodyDefinitions [{
      boundBone {boneName=dyn_jacket_side_r_02}
      boxExtents=<10 10 10>
      constraintSetup {angularLimitsMax=<0 5 20>, angularLimitsMin=<0 -5 0>, angularTarget=<0 0 0>}
      localJointOffset=<-1 0 0>
     }]
    sphericalLimits [
     {limitRadius=10, drivingBone {boneName=spine_04}, sphereLocalOffset=<0 0 1>}
     {limitRadius=10, drivingBone {boneName=spine_03}, sphereLocalOffset=<0 0 1>}
     {limitRadius=10, drivingBone {boneName=spine_02}, sphereLocalOffset=<0 0 1>}
    ]
   }
   n19 = Anim Dynamics - Bone: dyn_jacket_back_01(ComponentPose=@n18, Alpha=1, ExternalForce="0,0,0") {
    angularBiasOverride=0.999
    angularDampingOverride=0.999
    angularSpringConstant=300
    bAngularSpring=true
    bOverrideAngularBias=true
    bOverrideAngularDamping=true
    bUseGravityOverride=true
    bUsePlanarLimit=false
    bUseSphericalLimits=true
    simulationSpace=World
    alphaBoolBlend {blendOption=Linear}
    boundBone {boneName=dyn_jacket_back_01}
    componentAppliedLinearAccClamp=<600 600 600>
    componentLinearAccScale=<0.1 0.1 0.1>
    componentLinearVelScale=<0.1 0.1 0.1>
    gravityOverride {z=-980}
    physicsBodyDefinitions [{
      boundBone {boneName=dyn_jacket_back_01}
      boxExtents=<15 15 15>
      constraintSetup {angularLimitsMax=<0 5 30>, angularLimitsMin=<0 -5 -30>, angularTarget=<0 0 0>}
      localJointOffset=<1 0 0>
     }]
    sphericalLimits [
     {limitRadius=8, drivingBone {boneName=spine_04}, sphereLocalOffset=<0 0 -1>}
     {limitRadius=8, drivingBone {boneName=spine_03}, sphereLocalOffset=<0 0 -1>}
     {limitRadius=8, drivingBone {boneName=spine_02}, sphereLocalOffset=<0 0 -1>}
     {limitRadius=8, drivingBone {boneName=pelvis}, sphereLocalOffset=<0 0 0>}
    ]
   }
   n20 = Anim Dynamics - Bone: dyn_jacket_back_02(ComponentPose=@n19, Alpha=1, ExternalForce="0,0,0") {
    angularBiasOverride=0.999
    angularDampingOverride=0.999
    angularSpringConstant=300
    bAngularSpring=true
    bOverrideAngularBias=true
    bOverrideAngularDamping=true
    bUseGravityOverride=true
    bUsePlanarLimit=false
    bUseSphericalLimits=true
    simulationSpace=World
    alphaBoolBlend {blendOption=Linear}
    boundBone {boneName=dyn_jacket_back_02}
    componentAppliedLinearAccClamp=<600 600 600>
    componentLinearAccScale=<0.1 0.1 0.1>
    componentLinearVelScale=<0.1 0.1 0.1>
    gravityOverride {z=-980}
    physicsBodyDefinitions [{
      boundBone {boneName=dyn_jacket_back_02}
      boxExtents=<15 15 15>
      constraintSetup {angularLimitsMax=<0 5 30>, angularLimitsMin=<0 -5 -30>, angularTarget=<0 0 0>}
      localJointOffset=<1 0 0>
     }]
    sphericalLimits [
     {limitRadius=8, drivingBone {boneName=spine_04}, sphereLocalOffset=<0 0 -1>}
     {limitRadius=8, drivingBone {boneName=spine_03}, sphereLocalOffset=<0 0 -1>}
     {limitRadius=8, drivingBone {boneName=spine_02}, sphereLocalOffset=<0 0 -1>}
     {limitRadius=8, drivingBone {boneName=pelvis}, sphereLocalOffset=<0 0 0>}
    ]
   }
   n21 = Control Rig(Source=Use cached pose 'Base_pose'()) {
    alphaBoolBlend {blendOption=Linear}
    controlRigAssetReference {blueprintRigClass=&/F_MED_CoreStreet_CtrlRig.F_MED_CoreStreet_CtrlRig_C}
   }
   n22 = Anim Dynamics - Bone: dyn_neckalce_01(ComponentPose=Local To Component(LocalPose=@n21), Alpha=1, ExternalForce="0,0,0") {
    angularBiasOverride=0.999
    angularDampingOverride=0.99
    angularSpringConstant=300
    bAngularSpring=true
    bOverrideAngularBias=true
    bOverrideAngularDamping=true
    bUseGravityOverride=true
    alphaBoolBlend {blendOption=Linear}
    boundBone {boneName=dyn_neckalce_01}
    componentAppliedLinearAccClamp=<600 600 600>
    componentLinearAccScale=<0.1 0.1 0.1>
    componentLinearVelScale=<0.1 0.1 0.1>
    gravityOverride {z=-980}
    physicsBodyDefinitions [{
      boundBone {boneName=dyn_neckalce_01}
      boxExtents=<10 10 10>
      constraintSetup {angularLimitsMax=<0 0 10>, angularLimitsMin=<0 0 3>, angularTarget=<0 0 0>}
      localJointOffset=<1 0 0>
     }]
    planarLimits [{drivingBone {boneName=spine_05}, planeTransform {location=<13 0 0>, rotation=<rot 0 -30 90>}}]
   }
   n23 = Anim Dynamics - Bone: dyn_neckalce_02(ComponentPose=@n22, Alpha=1, ExternalForce="0,0,0") {
    angularBiasOverride=0.999
    angularDampingOverride=0.99
    angularSpringConstant=300
    bAngularSpring=true
    bOverrideAngularBias=true
    bOverrideAngularDamping=true
    bUseGravityOverride=true
    alphaBoolBlend {blendOption=Linear}
    boundBone {boneName=dyn_neckalce_02}
    componentAppliedLinearAccClamp=<600 600 600>
    componentLinearAccScale=<0.1 0.1 0.1>
    componentLinearVelScale=<0.1 0.1 0.1>
    gravityOverride {z=-980}
    physicsBodyDefinitions [{
      boundBone {boneName=dyn_neckalce_02}
      boxExtents=<10 10 10>
      constraintSetup {angularLimitsMax=<0 0 10>, angularLimitsMin=<0 0 3>, angularTarget=<0 0 0>}
      localJointOffset=<1 0 0>
     }]
    planarLimits [{drivingBone {boneName=spine_05}, planeTransform {location=<13 0 0>, rotation=<rot 0 -30 90>}}]
   }
   n24 = Anim Dynamics - Bone: dyn_neckalce_03(ComponentPose=@n23, Alpha=1, ExternalForce="0,0,0") {
    angularBiasOverride=0.999
    angularDampingOverride=0.99
    angularSpringConstant=300
    bAngularSpring=true
    bOverrideAngularBias=true
    bOverrideAngularDamping=true
    bUseGravityOverride=true
    alphaBoolBlend {blendOption=Linear}
    boundBone {boneName=dyn_neckalce_03}
    componentAppliedLinearAccClamp=<600 600 600>
    componentLinearAccScale=<0.1 0.1 0.1>
    componentLinearVelScale=<0.1 0.1 0.1>
    gravityOverride {z=-980}
    physicsBodyDefinitions [{
      boundBone {boneName=dyn_neckalce_03}
      boxExtents=<10 10 10>
      constraintSetup {angularLimitsMax=<0 10 10>, angularLimitsMin=<0 -10 3>, angularTarget=<0 0 0>, linearAxesMax=<0 0 0.5>, linearAxesMin=<0 0 -0.5>}
      localJointOffset=<1 0 0>
     }]
    planarLimits [{drivingBone {boneName=spine_05}, planeTransform {location=<13 0 0>, rotation=<rot 0 -30 90>}}]
   }
   n25 = Anim Dynamics - Bone: dyn_neckalce_04(ComponentPose=@n24, Alpha=1, ExternalForce="0,0,0") {
    angularBiasOverride=0.999
    angularDampingOverride=0.995
    angularSpringConstant=300
    bAngularSpring=true
    bOverrideAngularBias=true
    bOverrideAngularDamping=true
    bUseGravityOverride=true
    alphaBoolBlend {blendOption=Linear}
    boundBone {boneName=dyn_neckalce_04}
    componentAppliedLinearAccClamp=<600 600 600>
    componentLinearAccScale=<0.2 0.2 0.2>
    componentLinearVelScale=<0.2 0.2 0.2>
    gravityOverride {z=-980}
    physicsBodyDefinitions [{
      boundBone {boneName=dyn_neckalce_04}
      boxExtents=<9 9 9>
      constraintSetup {angularLimitsMax=<0 10 60>, angularLimitsMin=<0 -10 3>, angularTarget=<0 0 0>}
      localJointOffset=<1 0 0>
     }]
    planarLimits [{drivingBone {boneName=spine_05}, planeTransform {location=<13 0 0>, rotation=<rot 0 -25.4743 90>}}]
    sphericalLimits [
     {limitRadius=6.7416, drivingBone {boneName=deform_pec_l}, sphereLocalOffset=<-0.3528 -0.1821 0.7624>}
     {limitRadius=6.7416, drivingBone {boneName=deform_pec_r}, sphereLocalOffset=<0.4557 -0.1503 -0.6637>}
    ]
   }
   n26 = Anim Dynamics - Bone: dyn_jacket_fr_l_01(ComponentPose=@n25, Alpha=1, ExternalForce="0,0,0") {
    angularBiasOverride=0.999
    angularDampingOverride=0.99
    angularSpringConstant=300
    bAngularSpring=true
    bOverrideAngularBias=true
    bOverrideAngularDamping=true
    bUseGravityOverride=true
    bUsePlanarLimit=false
    alphaBoolBlend {blendOption=Linear}
    boundBone {boneName=dyn_jacket_fr_l_01}
    componentAppliedLinearAccClamp=<2000 2000 2000>
    componentLinearAccScale=<0.3 0.3 0.3>
    componentLinearVelScale=<0.3 0.3 0.3>
    gravityOverride {z=-980}
    physicsBodyDefinitions [{
      boundBone {boneName=dyn_jacket_fr_l_01}
      boxExtents=<10 10 10>
      constraintSetup {angularLimitsMax=<0 0 10>, angularLimitsMin=<0 0 0>, angularTarget=<0 0 0>}
      localJointOffset=<1 0 0>
     }]
    sphericalLimits [
     {limitRadius=10, drivingBone {boneName=spine_04}, sphereLocalOffset=<0 0 -1>}
     {limitRadius=10, drivingBone {boneName=spine_03}, sphereLocalOffset=<0 0 -1>}
     {limitRadius=10, drivingBone {boneName=spine_02}, sphereLocalOffset=<0 0 -1>}
    ]
   }
   n27 = Anim Dynamics - Bone: dyn_jacket_fr_l_02(ComponentPose=@n26, Alpha=1, ExternalForce="0,0,0") {
    angularBiasOverride=0.999
    angularDampingOverride=0.99
    angularSpringConstant=300
    bAngularSpring=true
    bOverrideAngularBias=true
    bOverrideAngularDamping=true
    bUseGravityOverride=true
    bUsePlanarLimit=false
    alphaBoolBlend {blendOption=Linear}
    boundBone {boneName=dyn_jacket_fr_l_02}
    componentAppliedLinearAccClamp=<2000 2000 2000>
    componentLinearAccScale=<0.45 0.45 0.45>
    componentLinearVelScale=<0.45 0.45 0.45>
    gravityOverride {z=-980}
    physicsBodyDefinitions [{
      boundBone {boneName=dyn_jacket_fr_l_02}
      boxExtents=<10 10 10>
      constraintSetup {angularLimitsMax=<0 10 30>, angularLimitsMin=<0 -10 0>, angularTarget=<0 0 0>}
      localJointOffset=<1 0 0>
     }]
    sphericalLimits [
     {limitRadius=10, drivingBone {boneName=spine_04}, sphereLocalOffset=<0 0 -1>}
     {limitRadius=10, drivingBone {boneName=spine_03}, sphereLocalOffset=<0 0 -1>}
     {limitRadius=10, drivingBone {boneName=spine_02}, sphereLocalOffset=<0 0 -1>}
    ]
   }
   n28 = Anim Dynamics - Bone: dyn_jacket_fr_l_03(ComponentPose=@n27, Alpha=1, ExternalForce="0,0,0") {
    angularBiasOverride=0.999
    angularDampingOverride=0.99
    angularSpringConstant=300
    bAngularSpring=true
    bOverrideAngularBias=true
    bOverrideAngularDamping=true
    bUseGravityOverride=true
    bUsePlanarLimit=false
    alphaBoolBlend {blendOption=Linear}
    boundBone {boneName=dyn_jacket_fr_l_03}
    componentAppliedLinearAccClamp=<2000 2000 2000>
    componentLinearAccScale=<0.5 0.5 0.5>
    componentLinearVelScale=<0.5 0.5 0.5>
    gravityOverride {z=-980}
    physicsBodyDefinitions [{
      boundBone {boneName=dyn_jacket_fr_l_03}
      boxExtents=<10 10 10>
      constraintSetup {angularLimitsMax=<0 10 50>, angularLimitsMin=<0 -10 -10>, angularTarget=<0 0 0>}
      localJointOffset=<1 0 0>
     }]
    sphericalLimits [
     {limitRadius=10, drivingBone {boneName=spine_04}, sphereLocalOffset=<0 0 -1>}
     {limitRadius=10, drivingBone {boneName=spine_03}, sphereLocalOffset=<0 0 -1>}
     {limitRadius=10, drivingBone {boneName=spine_02}, sphereLocalOffset=<0 0 -1>}
    ]
   }
   n29 = Anim Dynamics - Bone: dyn_jacket_fr_side_l_01(ComponentPose=@n28, Alpha=1, ExternalForce="0,0,0") {
    angularBiasOverride=0.999
    angularDampingOverride=0.99
    angularSpringConstant=300
    bAngularSpring=true
    bOverrideAngularBias=true
    bOverrideAngularDamping=true
    bUseGravityOverride=true
    bUsePlanarLimit=false
    bUseSphericalLimits=true
    alphaBoolBlend {blendOption=Linear}
    boundBone {boneName=dyn_jacket_fr_side_l_01}
    componentAppliedLinearAccClamp=<2000 2000 2000>
    componentLinearAccScale=<0.3 0.3 0.3>
    componentLinearVelScale=<0.3 0.3 0.3>
    gravityOverride {z=-980}
    physicsBodyDefinitions [{
      boundBone {boneName=dyn_jacket_fr_side_l_01}
      boxExtents=<10 10 10>
      constraintSetup {angularLimitsMax=<0 0 10>, angularLimitsMin=<0 0 0>, angularTarget=<0 0 0>}
      localJointOffset=<1 0 0>
     }]
    sphericalLimits [
     {limitRadius=10, drivingBone {boneName=spine_04}, sphereLocalOffset=<0 0 -1>}
     {limitRadius=10, drivingBone {boneName=spine_03}, sphereLocalOffset=<0 0 -1>}
     {limitRadius=10, drivingBone {boneName=spine_02}, sphereLocalOffset=<0 0 -1>}
     {limitRadius=13, drivingBone {boneName=pelvis}, sphereLocalOffset=<0 0 0>}
    ]
   }
   n30 = Anim Dynamics - Bone: dyn_jacket_fr_side_l_02(ComponentPose=@n29, Alpha=1, ExternalForce="0,0,0") {
    angularBiasOverride=0.999
    angularDampingOverride=0.99
    angularSpringConstant=300
    bAngularSpring=true
    bOverrideAngularBias=true
    bOverrideAngularDamping=true
    bUseGravityOverride=true
    bUsePlanarLimit=false
    bUseSphericalLimits=true
    alphaBoolBlend {blendOption=Linear}
    boundBone {boneName=dyn_jacket_fr_side_l_02}
    componentAppliedLinearAccClamp=<2000 2000 2000>
    componentLinearAccScale=<0.3 0.3 0.3>
    componentLinearVelScale=<0.3 0.3 0.3>
    gravityOverride {z=-980}
    physicsBodyDefinitions [{
      boundBone {boneName=dyn_jacket_fr_side_l_02}
      boxExtents=<10 10 10>
      constraintSetup {angularLimitsMax=<0 2 25>, angularLimitsMin=<0 -2 -5>, angularTarget=<0 0 0>}
      localJointOffset=<1 0 0>
     }]
    sphericalLimits [
     {limitRadius=10, drivingBone {boneName=spine_04}, sphereLocalOffset=<0 0 -1>}
     {limitRadius=10, drivingBone {boneName=spine_03}, sphereLocalOffset=<0 0 -1>}
     {limitRadius=10, drivingBone {boneName=spine_02}, sphereLocalOffset=<0 0 -1>}
     {limitRadius=13, drivingBone {boneName=pelvis}, sphereLocalOffset=<0 0 0>}
    ]
   }
   n31 = Anim Dynamics - Bone: dyn_jacket_side_l_01(ComponentPose=@n30, Alpha=1, ExternalForce="0,0,0") {
    angularBiasOverride=0.999
    angularDampingOverride=0.99
    angularSpringConstant=300
    bAngularSpring=true
    bOverrideAngularBias=true
    bOverrideAngularDamping=true
    bUseGravityOverride=true
    bUsePlanarLimit=false
    bUseSphericalLimits=true
    alphaBoolBlend {blendOption=Linear}
    boundBone {boneName=dyn_jacket_side_l_01}
    componentAppliedLinearAccClamp=<2000 2000 2000>
    componentLinearAccScale=<0.3 0.3 0.3>
    componentLinearVelScale=<0.3 0.3 0.3>
    gravityOverride {z=-980}
    physicsBodyDefinitions [{
      boundBone {boneName=dyn_jacket_side_l_01}
      boxExtents=<10 10 10>
      constraintSetup {angularLimitsMax=<0 0 10>, angularLimitsMin=<0 0 0>, angularTarget=<0 0 0>}
      localJointOffset=<1 0 0>
     }]
    sphericalLimits [
     {limitRadius=10, drivingBone {boneName=spine_04}, sphereLocalOffset=<0 0 -1>}
     {limitRadius=10, drivingBone {boneName=spine_03}, sphereLocalOffset=<0 0 -1>}
     {limitRadius=10, drivingBone {boneName=spine_02}, sphereLocalOffset=<0 0 -1>}
     {limitRadius=13, drivingBone {boneName=pelvis}, sphereLocalOffset=<0 0 0>}
    ]
   }
   n32 = Anim Dynamics - Bone: dyn_jacket_fr_side_l_02(ComponentPose=@n31, Alpha=1, ExternalForce="0,0,0") {
    angularBiasOverride=0.999
    angularDampingOverride=0.99
    angularSpringConstant=300
    bAngularSpring=true
    bOverrideAngularBias=true
    bOverrideAngularDamping=true
    bUseGravityOverride=true
    bUsePlanarLimit=false
    bUseSphericalLimits=true
    alphaBoolBlend {blendOption=Linear}
    boundBone {boneName=dyn_jacket_fr_side_l_02}
    componentAppliedLinearAccClamp=<2000 2000 2000>
    componentLinearAccScale=<0.3 0.3 0.3>
    componentLinearVelScale=<0.3 0.3 0.3>
    gravityOverride {z=-980}
    physicsBodyDefinitions [{
      boundBone {boneName=dyn_jacket_fr_side_l_02}
      boxExtents=<10 10 10>
      constraintSetup {angularLimitsMax=<0 0 20>, angularLimitsMin=<0 0 0>, angularTarget=<0 0 0>}
      localJointOffset=<1 0 0>
     }]
    sphericalLimits [
     {limitRadius=10, drivingBone {boneName=spine_04}, sphereLocalOffset=<0 0 -1>}
     {limitRadius=10, drivingBone {boneName=spine_03}, sphereLocalOffset=<0 0 -1>}
     {limitRadius=10, drivingBone {boneName=spine_02}, sphereLocalOffset=<0 0 -1>}
     {limitRadius=13, drivingBone {boneName=pelvis}, sphereLocalOffset=<0 0 0>}
    ]
   }
   n33 = Anim Dynamics - Bone: dyn_jacket_fr_r_01(ComponentPose=@n32, Alpha=1, ExternalForce="0,0,0") {
    angularBiasOverride=0.999
    angularDampingOverride=0.99
    angularSpringConstant=300
    bAngularSpring=true
    bOverrideAngularBias=true
    bOverrideAngularDamping=true
    bUseGravityOverride=true
    bUsePlanarLimit=false
    alphaBoolBlend {blendOption=Linear}
    boundBone {boneName=dyn_jacket_fr_r_01}
    componentAppliedLinearAccClamp=<2000 2000 2000>
    componentLinearAccScale=<0.3 0.3 0.3>
    componentLinearVelScale=<0.3 0.3 0.3>
    gravityOverride {z=-980}
    physicsBodyDefinitions [{
      boundBone {boneName=dyn_jacket_fr_r_01}
      boxExtents=<10 10 10>
      constraintSetup {angularLimitsMax=<0 1 10>, angularLimitsMin=<0 -1 -5>, angularTarget=<0 0 0>}
      localJointOffset=<-1 0 0>
     }]
    sphericalLimits [
     {limitRadius=10, drivingBone {boneName=spine_04}, sphereLocalOffset=<0 0 1>}
     {limitRadius=10, drivingBone {boneName=spine_03}, sphereLocalOffset=<0 0 1>}
     {limitRadius=10, drivingBone {boneName=spine_02}, sphereLocalOffset=<0 0 1>}
    ]
   }
   n34 = Anim Dynamics - Bone: dyn_jacket_fr_r_02(ComponentPose=@n33, Alpha=1, ExternalForce="0,0,0") {
    angularBiasOverride=0.999
    angularDampingOverride=0.99
    angularSpringConstant=300
    bAngularSpring=true
    bOverrideAngularBias=true
    bOverrideAngularDamping=true
    bUseGravityOverride=true
    bUsePlanarLimit=false
    alphaBoolBlend {blendOption=Linear}
    boundBone {boneName=dyn_jacket_fr_r_02}
    componentAppliedLinearAccClamp=<2000 2000 2000>
    componentLinearAccScale=<0.45 0.45 0.45>
    componentLinearVelScale=<0.45 0.45 0.45>
    gravityOverride {z=-980}
    physicsBodyDefinitions [{
      boundBone {boneName=dyn_jacket_fr_r_02}
      boxExtents=<10 10 10>
      constraintSetup {angularLimitsMax=<0 10 30>, angularLimitsMin=<0 -10 -5>, angularTarget=<0 0 0>}
      localJointOffset=<-1 0 0>
     }]
    sphericalLimits [
     {limitRadius=10, drivingBone {boneName=spine_04}, sphereLocalOffset=<0 0 1>}
     {limitRadius=10, drivingBone {boneName=spine_03}, sphereLocalOffset=<0 0 1>}
     {limitRadius=10, drivingBone {boneName=spine_02}, sphereLocalOffset=<0 0 1>}
    ]
   }
   n35 = Anim Dynamics - Bone: dyn_jacket_fr_r_03(ComponentPose=@n34, Alpha=1, ExternalForce="0,0,0") {
    angularBiasOverride=0.999
    angularDampingOverride=0.99
    angularSpringConstant=300
    bAngularSpring=true
    bOverrideAngularBias=true
    bOverrideAngularDamping=true
    bUseGravityOverride=true
    bUsePlanarLimit=false
    alphaBoolBlend {blendOption=Linear}
    boundBone {boneName=dyn_jacket_fr_r_03}
    componentAppliedLinearAccClamp=<2000 2000 2000>
    componentLinearAccScale=<0.5 0.5 0.5>
    componentLinearVelScale=<0.5 0.5 0.5>
    gravityOverride {z=-980}
    physicsBodyDefinitions [{
      boundBone {boneName=dyn_jacket_fr_r_03}
      boxExtents=<10 10 10>
      constraintSetup {angularLimitsMax=<0 10 50>, angularLimitsMin=<0 -10 -5>, angularTarget=<0 0 0>}
      localJointOffset=<-1 0 0>
     }]
    sphericalLimits [
     {limitRadius=10, drivingBone {boneName=spine_04}, sphereLocalOffset=<0 0 1>}
     {limitRadius=10, drivingBone {boneName=spine_03}, sphereLocalOffset=<0 0 1>}
     {limitRadius=10, drivingBone {boneName=spine_02}, sphereLocalOffset=<0 0 1>}
    ]
   }
   n36 = Anim Dynamics - Bone: dyn_jacket_fr_side_r_01(ComponentPose=@n35, Alpha=1, ExternalForce="0,0,0") {
    angularBiasOverride=0.999
    angularDampingOverride=0.99
    angularSpringConstant=300
    bAngularSpring=true
    bOverrideAngularBias=true
    bOverrideAngularDamping=true
    bUseGravityOverride=true
    bUsePlanarLimit=false
    alphaBoolBlend {blendOption=Linear}
    boundBone {boneName=dyn_jacket_fr_side_r_01}
    componentAppliedLinearAccClamp=<2000 2000 2000>
    componentLinearAccScale=<0.3 0.3 0.3>
    componentLinearVelScale=<0.3 0.3 0.3>
    gravityOverride {z=-980}
    physicsBodyDefinitions [{
      boundBone {boneName=dyn_jacket_fr_side_r_01}
      boxExtents=<10 10 10>
      constraintSetup {angularLimitsMax=<0 1 10>, angularLimitsMin=<0 -1 -5>, angularTarget=<0 0 0>}
      localJointOffset=<-1 0 0>
     }]
    sphericalLimits [
     {limitRadius=10, drivingBone {boneName=spine_04}, sphereLocalOffset=<0 0 1>}
     {limitRadius=10, drivingBone {boneName=spine_03}, sphereLocalOffset=<0 0 1>}
     {limitRadius=10, drivingBone {boneName=spine_02}, sphereLocalOffset=<0 0 1>}
    ]
   }
   n37 = Anim Dynamics - Bone: dyn_jacket_fr_side_r_02(ComponentPose=@n36, Alpha=1, ExternalForce="0,0,0") {
    angularBiasOverride=0.999
    angularDampingOverride=0.99
    angularSpringConstant=300
    bAngularSpring=true
    bOverrideAngularBias=true
    bOverrideAngularDamping=true
    bUseGravityOverride=true
    bUsePlanarLimit=false
    alphaBoolBlend {blendOption=Linear}
    boundBone {boneName=dyn_jacket_fr_side_r_02}
    componentAppliedLinearAccClamp=<2000 2000 2000>
    componentLinearAccScale=<0.3 0.3 0.3>
    componentLinearVelScale=<0.3 0.3 0.3>
    gravityOverride {z=-980}
    physicsBodyDefinitions [{
      boundBone {boneName=dyn_jacket_fr_side_r_02}
      boxExtents=<10 10 10>
      constraintSetup {angularLimitsMax=<0 2 25>, angularLimitsMin=<0 -2 -5>, angularTarget=<0 0 0>}
      localJointOffset=<-1 0 0>
     }]
    sphericalLimits [
     {limitRadius=10, drivingBone {boneName=spine_04}, sphereLocalOffset=<0 0 1>}
     {limitRadius=10, drivingBone {boneName=spine_03}, sphereLocalOffset=<0 0 1>}
     {limitRadius=10, drivingBone {boneName=spine_02}, sphereLocalOffset=<0 0 1>}
    ]
   }
   n38 = Anim Dynamics - Bone: dyn_jacket_side_r_01(ComponentPose=@n37, Alpha=1, ExternalForce="0,0,0") {
    angularBiasOverride=0.999
    angularDampingOverride=0.99
    angularSpringConstant=300
    bAngularSpring=true
    bOverrideAngularBias=true
    bOverrideAngularDamping=true
    bUseGravityOverride=true
    bUsePlanarLimit=false
    alphaBoolBlend {blendOption=Linear}
    boundBone {boneName=dyn_jacket_side_r_01}
    componentAppliedLinearAccClamp=<2000 2000 2000>
    componentLinearAccScale=<0.3 0.3 0.3>
    componentLinearVelScale=<0.3 0.3 0.3>
    gravityOverride {z=-980}
    physicsBodyDefinitions [{
      boundBone {boneName=dyn_jacket_side_r_01}
      boxExtents=<10 10 10>
      constraintSetup {angularLimitsMax=<0 5 30>, angularLimitsMin=<0 -5 -5>, angularTarget=<0 0 0>}
      localJointOffset=<-1 0 0>
     }]
    sphericalLimits [
     {limitRadius=10, drivingBone {boneName=spine_04}, sphereLocalOffset=<0 0 1>}
     {limitRadius=10, drivingBone {boneName=spine_03}, sphereLocalOffset=<0 0 1>}
     {limitRadius=10, drivingBone {boneName=spine_02}, sphereLocalOffset=<0 0 1>}
    ]
   }
   n39 = Anim Dynamics - Bone: dyn_jacket_side_r_02(ComponentPose=@n38, Alpha=1, ExternalForce="0,0,0") {
    angularBiasOverride=0.999
    angularDampingOverride=0.99
    angularSpringConstant=300
    bAngularSpring=true
    bOverrideAngularBias=true
    bOverrideAngularDamping=true
    bUseGravityOverride=true
    bUsePlanarLimit=false
    alphaBoolBlend {blendOption=Linear}
    boundBone {boneName=dyn_jacket_side_r_02}
    componentAppliedLinearAccClamp=<2000 2000 2000>
    componentLinearAccScale=<0.3 0.3 0.3>
    componentLinearVelScale=<0.3 0.3 0.3>
    gravityOverride {z=-980}
    physicsBodyDefinitions [{
      boundBone {boneName=dyn_jacket_side_r_02}
      boxExtents=<10 10 10>
      constraintSetup {angularLimitsMax=<0 1 20>, angularLimitsMin=<0 -1 -5>, angularTarget=<0 0 0>}
      localJointOffset=<-1 0 0>
     }]
    sphericalLimits [
     {limitRadius=10, drivingBone {boneName=spine_04}, sphereLocalOffset=<0 0 1>}
     {limitRadius=10, drivingBone {boneName=spine_03}, sphereLocalOffset=<0 0 1>}
     {limitRadius=10, drivingBone {boneName=spine_02}, sphereLocalOffset=<0 0 1>}
    ]
   }
   n40 = Anim Dynamics - Bone: dyn_jacket_back_01(ComponentPose=@n39, Alpha=1, ExternalForce="0,0,0") {
    angularBiasOverride=0.999
    angularDampingOverride=0.99
    angularSpringConstant=300
    bAngularSpring=true
    bOverrideAngularDamping=true
    bUseGravityOverride=true
    bUsePlanarLimit=false
    bUseSphericalLimits=true
    alphaBoolBlend {blendOption=Linear}
    boundBone {boneName=dyn_jacket_back_01}
    componentAppliedLinearAccClamp=<2000 2000 2000>
    componentLinearAccScale=<0.3 0.3 0.3>
    componentLinearVelScale=<0.3 0.3 0.3>
    gravityOverride {z=-980}
    physicsBodyDefinitions [{
      boundBone {boneName=dyn_jacket_back_01}
      boxExtents=<12 12 12>
      constraintSetup {angularLimitsMax=<0 0 30>, angularLimitsMin=<0 0 -30>, angularTarget=<0 0 0>}
      localJointOffset=<1 0 0>
     }]
    sphericalLimits [
     {limitRadius=8, drivingBone {boneName=spine_04}, sphereLocalOffset=<0 0 -1>}
     {limitRadius=8, drivingBone {boneName=spine_03}, sphereLocalOffset=<0 0 -1>}
     {limitRadius=8, drivingBone {boneName=spine_02}, sphereLocalOffset=<0 0 -1>}
     {limitRadius=8, drivingBone {boneName=pelvis}, sphereLocalOffset=<0 0 0>}
    ]
   }
   n41 = Anim Dynamics - Bone: dyn_jacket_back_02(ComponentPose=@n40, Alpha=1, ExternalForce="0,0,0") {
    angularBiasOverride=0.999
    angularDampingOverride=0.99
    angularSpringConstant=300
    bAngularSpring=true
    bOverrideAngularDamping=true
    bUseGravityOverride=true
    bUsePlanarLimit=false
    bUseSphericalLimits=true
    alphaBoolBlend {blendOption=Linear}
    boundBone {boneName=dyn_jacket_back_02}
    componentAppliedLinearAccClamp=<2000 2000 2000>
    componentLinearAccScale=<0.3 0.3 0.3>
    componentLinearVelScale=<0.3 0.3 0.3>
    gravityOverride {z=-980}
    physicsBodyDefinitions [{
      boundBone {boneName=dyn_jacket_back_02}
      boxExtents=<12 12 12>
      constraintSetup {angularLimitsMax=<0 0 30>, angularLimitsMin=<0 -5 -30>, angularTarget=<0 0 0>}
      localJointOffset=<1 0 0>
     }]
    sphericalLimits [
     {limitRadius=8, drivingBone {boneName=spine_04}, sphereLocalOffset=<0 0 -1>}
     {limitRadius=8, drivingBone {boneName=spine_03}, sphereLocalOffset=<0 0 -1>}
     {limitRadius=8, drivingBone {boneName=spine_02}, sphereLocalOffset=<0 0 -1>}
     {limitRadius=8, drivingBone {boneName=pelvis}, sphereLocalOffset=<0 0 0>}
    ]
   }
   Output Pose(Result=Blend Poses by bool(BlendPose_0=Component To Local(ComponentPose=@n20), BlendPose_1=Component To Local(ComponentPose=@n41)) {blendType=Linear}) {layerGroup=None}
   Save cached pose 'Base_pose'(Pose=Copy Pose From Mesh() {bUseAttachedParent=true}) {cacheName=Base_pose}
  ]}
 EventGraph {nodes [Event BlueprintUpdateAnimation()]}
]
variables []
